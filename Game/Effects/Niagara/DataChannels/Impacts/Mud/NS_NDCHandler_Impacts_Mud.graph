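# Niagara system that handles mud impact effects from the Fort impacts data channel, with a brief warm flash, mud splash mesh, splash sprites and falling drops.
# Emitters spawn via Spawn_NDCFortImpacts, then read impact point, normal, damage and LOD distance per particle.
# Damage scales size and launch speed; LOD distance enlarges the distant flash, while filters limit drops and splash sprites.
# Light and Debug emitters are disabled.
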NiagaraSystem /Game/Effects/Niagara/DataChannels/Impacts/Mud/NS_NDCHandler_Impacts_Mud
dependencies {
 usedDataInterfaces [
  &NiagaraDataInterfaceColorCurve
  &NiagaraDataInterfaceCurve
  /Script/Niagara.NiagaraDataInterfaceDataChannelRead
  &NiagaraDataInterfaceSpriteRendererInfo
 ]
 usedDynamicInputs [&RotateVectorByQuaternion]
 usedModules [
  &self:SystemSpawnScript.SystemScriptSource.SystemScriptGraph.NiagaraNodeAssignment_0.SetVariables_#1
  &SystemState
  &self:SystemSpawnScript.SystemScriptSource.SystemScriptGraph.NiagaraNodeAssignment_1.SetVariables_#2
  /Niagara/Modules/System/CompleteIfUnused
 ]
}
emitters {
 Debug {
  bEnabled=false
  emitterUpdateScript [
   &EmitterState()
   /Game/Effects/Niagara/DataChannels/Impacts/Spawn_NDCFortImpacts(
    Max Spawn Count=1
    Min Spawn Count=1
    Reader=DataInterface() {propertyValues {
      Channel=/Game/Effects/Niagara/DataChannels/Impacts/NDC_FortImpacts
      bAutoLinkToSpawningNDC=true
      bOnlySpawnOnceOnSubticks=true
      bOverrideSpawnGroupToDataChannelIndex=true
      bReadCurrentFrame=true
      bUpdateSourceDataEveryTick=true
      AccessContext {accessContext {}}
     }}
   )
  ]
  particleSpawnScript [
   /Game/Effects/Niagara/DataChannels/Impacts/InitializeParticleFromNDC_FortImpact() {enabled=false}
   &/Read_NDC_FortImpacts_Mud()
   &InitializeParticle(
    Color=&MakeLinearColor(
     R:NiagaraFloat=$Output.InitializeParticleFromNDC_FortImpact.LODDistance
     G:NiagaraFloat=$Output.InitializeParticleFromNDC_FortImpact.ImpactDamage
     B:NiagaraFloat=&MakeCustomFloatFromBool(Bool=$Output.InitializeParticleFromNDC_FortImpact.bSimple, True Float=1, False Float=0)
     A:NiagaraFloat=&MakeFloatFromInt(Int=$Output.InitializeParticleFromNDC_FortImpact.ImpactPhysicalSurfaceInteger)
    )
    Position=$Output.InitializeParticleFromNDC_FortImpact.ImpactPoint
    Position Mode=::NewEnumerator1
    Sprite Size Mode=::NewEnumerator3
    Uniform Sprite Size=100
    UsePositionOffset=false
    Write Material Random=true
   )
  ]
  particleUpdateScript [
   &ParticleState()
   &CameraOffset(Camera Offset Amount=50)
  ]
  rendererClasses [&NiagaraSpriteRendererProperties]
  renderers [{rendererClass=&NiagaraSpriteRendererProperties}]
 }
 Flash {
  emitterUpdateScript [
   &EmitterState()
   &Spawn_NDCFortImpacts(Max Spawn Count=1, Min Spawn Count=1, Reader:NiagaraDataInterfaceDataChannelRead=$System.NDCReader)
  ]
  particleSpawnScript [
   &Read_NDC_FortImpacts_Mud()
   &InitializeParticleFromNDC_FortImpact() {enabled=false}
   &InitializeParticle(
    Lifetime=0.2
    Mesh Scale=&MakeVector(
     X=&Multiply_Float(
      A=&RandomRangeFloat(Minimum=1.8, Maximum=2)
      B=&RemapRange(Input Value To Remap=$Output.InitializeParticleFromNDC_FortImpact.ImpactDamage, Input Min=0, Input Max=100, Output Min=0.5, Output Max=1.5)
     )
     Y=&RandomRangeFloat(Minimum=0.6, Maximum=1)
     Z=&RandomRangeFloat(Minimum=0.6, Maximum=1)
    )
    Mesh Scale Mode=::NewEnumerator4
    Position=$Output.InitializeParticleFromNDC_FortImpact.ImpactPoint
    Position Mode=::NewEnumerator1
    Sprite Size Mode=::NewEnumerator3
    Uniform Sprite Size=&Multiply_Float(
     A=&RandomRangeFloat(Minimum=200, Maximum=220)
     B=&RemapRange(Input Value To Remap=$Output.InitializeParticleFromNDC_FortImpact.ImpactDamage, Input Min=10, Input Max=100, Output Min=0.5, Output Max=1.5)
    )
    UsePositionOffset=false
    Write Material Random=true
   )
   /SetVariables(Particles.AngleRange=0.02) {isSetParameters=true}
   &SubUVAnimation(End Frame=3, SubUV Animation Mode:ENiagara_SubUVLookupMode=::NewEnumerator2)
   &CameraOffset(Camera Offset Amount=10)
   &InitialMeshOrientation(
    Mesh Orientation Mode=::NewEnumerator2
    Orientation Vector=$Output.InitializeParticleFromNDC_FortImpact.ImpactNormal
    Rotation=&MakeVector(
     X=&RandomRangeFloat(Minimum=0, Maximum=1)
     Y=&Multiply_Float(A=$Particles.AngleRange, B=&RandomRangeFloat(Minimum=-1, Maximum=1))
     Z=&Multiply_Float(A=$Particles.AngleRange, B=&RandomRangeFloat(Minimum=-1, Maximum=1))
    )
    Use Rotation Vector=true
   )
  ]
  particleUpdateScript [
   &ParticleState()
   &Color(
    Color=&MakeLinearColorFromVectorAndFloat(
     "Vector (RGB)"=&MakeVectorFromLinearColorRGB(Color=&ColorFromCurve(ColorCurve=DataInterface() {propertyValues {
         AlphaCurve {keys []}
         BlueCurve {keys [
           {arriveTangent=-1.4036, interpMode=RCIM_Cubic, leaveTangent=-1.4036, time=0, value=1.4062}
           {arriveTangent=-1.4036, interpMode=RCIM_Cubic, leaveTangent=-1.4036, time=0.9977, value=0.0059}
          ]}
         GreenCurve {keys [
           {arriveTangent=-5.857, interpMode=RCIM_Cubic, leaveTangent=-5.857, time=0, value=5.8496}
           {arriveTangent=-5.857, interpMode=RCIM_Cubic, leaveTangent=-5.857, time=0.9977, value=0.0059}
          ]}
         RedCurve {keys [
           {arriveTangent=-14.4701, interpMode=RCIM_Cubic, leaveTangent=-14.4701, time=0, value=15}
           {arriveTangent=-14.4701, interpMode=RCIM_Cubic, leaveTangent=-14.4701, time=0.9977, value=0.5625}
          ]}
        }}))
     "Float (Alpha)"=&FloatFromCurve(FloatCurve=DataInterface() {propertyValues {Curve {keys [{interpMode=RCIM_Cubic, tangentMode=RCTM_User, time=0, value=1}, {interpMode=RCIM_Cubic, tangentMode=RCTM_User, time=1, value=0}]}}})
    )
    Scale Color=&VectorFromFloat(Value=&FloatFromCurve(FloatCurve=DataInterface() {propertyValues {Curve {keys [{interpMode=RCIM_Cubic, tangentMode=RCTM_User, time=0, value=1}, {interpMode=RCIM_Cubic, tangentMode=RCTM_User, time=1, value=0}]}}}))
   )
   &ScaleMeshSize(Scale Factor=&VectorFromFloat(Value=&Multiply_Float(
      A=&FloatFromCurve(
       FloatCurve=DataInterface() {propertyValues {Curve {keys [
           {interpMode=RCIM_Cubic, tangentMode=RCTM_Break, time=0, value=0.4}
           {interpMode=RCIM_Cubic, tangentMode=RCTM_Break, time=0.7, value=1}
           <at 1 1>
          ]}}}
       CurveIndex=&RemapRange(Input Value To Remap=$Output.InitializeParticleFromNDC_FortImpact.LODDistance, Input Min=100, Input Max=8000, Output Min=0, Output Max=1)
      )
      B=&FloatFromCurve(FloatCurve=DataInterface() {propertyValues {Curve {keys [{arriveTangent=1.9165, interpMode=RCIM_Cubic, leaveTangent=1.9165, tangentMode=RCTM_User, time=0, value=0.25}, <at 1 1>]}}})
     )))
   &FN_Sprite_Attributes(
    Create SpriteSize Attribute=true
    SpriteSize=&Multiply_Vector2D_ByFloat(
     Float=&FloatFromCurve(
      FloatCurve=DataInterface() {propertyValues {Curve {keys [
          {interpMode=RCIM_Cubic, tangentMode=RCTM_Break, time=0, value=0.3}
          {interpMode=RCIM_Cubic, tangentMode=RCTM_Break, time=0.7, value=1}
          <at 1 1>
         ]}}}
      CurveIndex=&RemapRange(Input Value To Remap=$Output.InitializeParticleFromNDC_FortImpact.LODDistance, Input Min=100, Input Max=8000, Output Min=0, Output Max=1)
     )
     Vector2D=&Multiply_Vector2D_ByFloat(
      Float=&FloatFromCurve(FloatCurve=DataInterface() {propertyValues {Curve {keys [{arriveTangent=1.9165, interpMode=RCIM_Cubic, leaveTangent=1.9165, tangentMode=RCTM_User, time=0, value=0.25}, <at 1 1>]}}})
      Vector2D=$Particles.SpriteSize
     )
    )
    Write to Module Namespace=true
   ) {module=FN_Sprite}
   /SetVariables(Particles.SpriteColor=&MakeLinearColorFromVectorAndFloat(
     "Vector (RGB)"=&MakeVectorFromLinearColorRGB(Color=&ColorFromCurve(ColorCurve=DataInterface() {propertyValues {
         AlphaCurve {keys []}
         BlueCurve {keys [
           {arriveTangent=-0.4071, interpMode=RCIM_Cubic, leaveTangent=-0.4071, time=0, value=0.6562}
           {arriveTangent=-0.4071, interpMode=RCIM_Cubic, leaveTangent=-0.4071, time=1, value=0.2491}
          ]}
         GreenCurve {keys [
           {arriveTangent=-2.6796, interpMode=RCIM_Cubic, leaveTangent=-2.6796, time=0, value=2.9487}
           {arriveTangent=-2.6796, interpMode=RCIM_Cubic, leaveTangent=-2.6796, time=1, value=0.2691}
          ]}
         RedCurve {keys [
           {arriveTangent=-6.5, interpMode=RCIM_Cubic, leaveTangent=-6.5, time=0, value=7}
           {arriveTangent=-6.5, interpMode=RCIM_Cubic, leaveTangent=-6.5, time=1, value=0.5}
          ]}
        }}))
     "Float (Alpha)"=&FloatFromCurve(FloatCurve=DataInterface() {propertyValues {Curve {keys [{interpMode=RCIM_Cubic, tangentMode=RCTM_Break, time=0, value=1}, {interpMode=RCIM_Cubic, tangentMode=RCTM_Break, time=1, value=0}]}}})
    )) {isSetParameters=true}
   &DynamicMaterialParameters(Write Parameter Index 0=false)
  ]
  rendererClasses [&NiagaraMeshRendererProperties, &NiagaraSpriteRendererProperties]
  renderers [
   {rendererClass=&NiagaraMeshRendererProperties, rendererIndex=0}
   {rendererClass=&NiagaraSpriteRendererProperties, rendererIndex=1}
  ]
 }
 Light {
  bEnabled=false
  emitterUpdateScript [
   &EmitterState()
   &Spawn_NDCFortImpacts(Max Spawn Count=1, Min Spawn Count=1, Reader:NiagaraDataInterfaceDataChannelRead=$System.NDCReader)
   /Game/Effects/Niagara/DataChannels/Impacts/SimpleFilter_NDCFortImpacts(Exclude=false) {enabled=false}
   /Game/Effects/Niagara/DataChannels/Impacts/LODDistanceFilter_NDCFortImpacts(LODDistance=2000)
  ]
  particleSpawnScript [
   &InitializeParticleFromNDC_FortImpact() {enabled=false}
   &Read_NDC_FortImpacts_Mud()
   &InitializeParticle(
    Lifetime=0.17
    Position=$Output.InitializeParticleFromNDC_FortImpact.ImpactPoint
    Position Mode=::NewEnumerator1
    UsePositionOffset=false
    Write Material Random=true
   )
   /SetVariables(Particles.LightPositionNew:Vector3f=&Add_Vector(
     A=Linked() {linkedVariable {name=Particles.Position, type=NiagaraPosition}}
     B=&Multiply_VectorByFloat(Vector=$Output.InitializeParticleFromNDC_FortImpact.ImpactNormal, Float=25)
    )) {isSetParameters=true}
   /SetVariables(Particles.LightExponent=3) {isSetParameters=true}
   /SetVariables(Particles.LightRadius=&Multiply_Float(
     A=&Multiply_Float(
      A=60
      B=&RemapRange(Input Value To Remap=$Output.InitializeParticleFromNDC_FortImpact.LODDistance, Input Min=100, Input Max=8000, Output Min=1, Output Max=2)
     )
     B=&RemapRange(Input Value To Remap=$Output.InitializeParticleFromNDC_FortImpact.ImpactDamage, Input Min=0, Input Max=100, Output Min=0.5, Output Max=1.5)
    )) {isSetParameters=true}
  ]
  particleUpdateScript [
   &ParticleState()
   /SetVariables(Particles.Light_Color:LinearColor=&MultiplyLinearColor(
     Linear Color=<rgba 1 0.3184 0.1481 1>
     Scale Factor=&MakeFloat4FromVectorAndFloat(
      "Vector (RGB)"=<1 1 1>
      "Float (Alpha)"=&Multiply_Float(
       A=5
       B=&FloatFromCurve(
        FloatCurve=DataInterface() {propertyValues {Curve {keys [
            {interpMode=RCIM_Cubic, tangentMode=RCTM_User, time=0, value=0}
            {interpMode=RCIM_Cubic, tangentMode=RCTM_User, time=0.2, value=1}
            {tangentMode=RCTM_User, time=1, value=0}
           ]}}}
        Scale Curve=&FloatFromCurve(
         FloatCurve=DataInterface() {propertyValues {Curve {keys [{interpMode=RCIM_Cubic, tangentMode=RCTM_User, time=0, value=1}, {interpMode=RCIM_Cubic, tangentMode=RCTM_User, time=1, value=0}]}}}
         CurveIndex=&RemapRange(Input Value To Remap=$Output.InitializeParticleFromNDC_FortImpact.LODDistance, Input Min=0, Input Max=2000, Output Min=0, Output Max=1)
        )
       )
      )
     )
    )) {isSetParameters=true}
  ]
  rendererClasses [&NiagaraLightRendererProperties]
  renderers [{rendererClass=&NiagaraLightRendererProperties, rendererIndex=0}]
 }
 NE_Parent_Mud_DropsCollision {
  emitterUpdateScript [
   &EmitterState()
   &Spawn_NDCFortImpacts(Max Spawn Count=9, Min Spawn Count=5, Reader:NiagaraDataInterfaceDataChannelRead=$System.NDCReader)
   &SimpleFilter_NDCFortImpacts(Exclude=false, Max Spawn Count=3, Min Spawn Count=2)
   &LODDistanceFilter_NDCFortImpacts(LODDistance=3000)
  ]
  particleSpawnScript [
   &InitializeParticleFromNDC_FortImpact() {enabled=false}
   &Read_NDC_FortImpacts_Mud()
   &InitializeParticle(
    Lifetime Max=0.6
    Lifetime Min=0.4
    Lifetime Mode=::NewEnumerator1
    Position=$Output.InitializeParticleFromNDC_FortImpact.ImpactPoint
    Position Mode=::NewEnumerator1
    Sprite Size Mode=::NewEnumerator3
    Uniform Sprite Size=&Multiply_Float(
     A=&RandomRangeFloat(Minimum=3, Maximum=5)
     B=&RemapRange(Input Value To Remap=$Output.InitializeParticleFromNDC_FortImpact.ImpactDamage, Input Min=10, Input Max=100, Output Min=0.9, Output Max=1.1)
    )
    UsePositionOffset=false
    Write Material Random=true
   )
   &ShapeLocation(
    MaskSpawnGroup=false
    Offset=<5 0 0>
    Offset Coordinate Space=::World
    Offset Mode=::NewEnumerator0
    Override Seed=false
    Rotation Coordinate Space=::World
    Rotation Mode=::NewEnumerator1
    Sphere Radius=2
    Sphere Surface Distribution=1
    UseOwnerScale=true
   ) {module=ShapeLocation001}
   &AddVelocity(Rotation Coordinate Space=::World, Velocity=$Particles.ShapeLocation001.ShapeNormal, Velocity Speed Scale=100) {module=AddVelocity001}
   &AddVelocity(
    Velocity=$Output.InitializeParticleFromNDC_FortImpact.ImpactNormal
    Velocity Speed Scale=&Multiply_Float(
     A=&RandomRangeFloat(Minimum=200, Maximum=400)
     B=&RemapRange(
      Input Value To Remap=$Output.InitializeParticleFromNDC_FortImpact.ImpactDamage
      Input Min=10
      Input Max=100
      Output Min=0.75
      Output Max=1.25
     )
    )
   )
   &ApplyInitialForces(Apply Rotational Force to Rotational Velocity=false)
  ]
  particleUpdateScript [
   &ParticleState()
   &GravityForce()
   &SolveForcesAndVelocity(Clamp Velocity=false, Limit Acceleration=false)
   &DynamicMaterialParameters(
    Index 0 Param 1=$Particles.Timeline.Float
    Index 0 Param 2=5
    Index 0 Param 3=10
    Index 0 Param 4=5
    Module.Param0WriteEnabled=true
    Module.Param1WriteEnabled=true
    Module.Param2WriteEnabled=true
    Module.Param3WriteEnabled=true
   )
   &ScaleColor(Scale RGB=<0 0 0>, ScaleA=false, ScaleRGB=true)
   &ScaleSpriteSize(Uniform Curve Sprite Scale=DataInterface() {propertyValues {Curve {keys [
        {interpMode=RCIM_Cubic, tangentMode=RCTM_User, time=0.7, value=1}
        {arriveTangent=-6.3247, interpMode=RCIM_Cubic, leaveTangent=-6.3247, tangentMode=RCTM_User, time=1, value=0}
       ]}}})
  ]
  rendererClasses [&NiagaraSpriteRendererProperties]
  renderers [{rendererClass=&NiagaraSpriteRendererProperties}]
 }
 NE_Parent_Mud_Splash {
  emitterUpdateScript [
   &EmitterState()
   &Spawn_NDCFortImpacts(Max Spawn Count=1, Min Spawn Count=1, Reader:NiagaraDataInterfaceDataChannelRead=$System.NDCReader)
  ]
  particleSpawnScript [
   &InitializeParticleFromNDC_FortImpact() {enabled=false}
   &Read_NDC_FortImpacts_Mud()
   &InitializeParticle(
    Color=<rgba 1 1 1 1.2>
    Lifetime=0.25
    Mesh Scale Mode=::NewEnumerator3
    Mesh Uniform Scale=&Multiply_Float(
     A=&RandomRangeFloat(Minimum=2, Maximum=2.5)
     B=&RemapRange(
      Input Value To Remap=$Output.InitializeParticleFromNDC_FortImpact.ImpactDamage
      Input Min=10
      Input Max=100
      Output Min=0.75
      Output Max=1.25
     )
    )
    Position=$Output.InitializeParticleFromNDC_FortImpact.ImpactPoint
    Position Mode=::NewEnumerator1
    Sprite Rotation Mode=::NewEnumerator2
    Sprite UV Mode=::NewEnumerator3
    UsePositionOffset=false
    Write Material Random=true
   )
   &InitialMeshOrientation(
    Mesh Axis To Orient=<0 0 1>
    Mesh Orientation Mode=::NewEnumerator2
    Orientation Vector=$Output.InitializeParticleFromNDC_FortImpact.ImpactNormal
    Rotation=&RandomRangeVector(Minimum=<0 0 0>, Maximum=<0 0 360>)
    Use Rotation Vector=true
   )
   &AddVelocity(Rotation Coordinate Space=::World, Velocity=$Output.InitializeParticleFromNDC_FortImpact.ImpactNormal, Velocity Speed Scale=25) {module=AddVelocity001}
   &CameraOffset(Camera Offset Amount=15)
  ]
  particleUpdateScript [
   &ParticleState()
   &ScaleMeshSize(Scale Factor=&Multiply_VectorByFloat(
     Vector=&MakeVector(
      X=1
      Y=1
      Z=&FloatFromCurve(
       FloatCurve=DataInterface() {propertyValues {Curve {keys [
           {tangentMode=RCTM_User, time=0, value=1.875}
           {arriveTangent=-0.0405, interpMode=RCIM_Cubic, leaveTangent=-0.0405, tangentMode=RCTM_User, time=0.5, value=0.875}
           {arriveTangent=-3.7129, interpMode=RCIM_Cubic, leaveTangent=-3.7129, tangentMode=RCTM_User, time=1, value=0}
          ]}}}
       Scale Curve=1.25
      )
     )
     Float=&FloatFromCurve(
      FloatCurve=DataInterface() {propertyValues {Curve {keys [
          {arriveTangent=3.4838, interpMode=RCIM_Cubic, leaveTangent=3.4838, tangentMode=RCTM_User, time=0, value=0.2}
          {arriveTangent=1.1179, interpMode=RCIM_Cubic, leaveTangent=1.1179, tangentMode=RCTM_User, time=0.2, value=0.6}
          {arriveTangent=0.0224, interpMode=RCIM_Cubic, leaveTangent=0.0224, tangentMode=RCTM_User, time=1, value=1}
         ]}}}
      Scale Curve=1.05
     )
    ))
   &GravityForce(Gravity=<0 0 -300>)
   &SolveForcesAndVelocity(Clamp Velocity=false, Limit Acceleration=false)
   &ScaleSpriteSize(Uniform Curve Sprite Scale=DataInterface() {propertyValues {Curve {keys [
        {arriveTangent=4.1785, interpMode=RCIM_Cubic, leaveTangent=4.1785, tangentMode=RCTM_User, time=0, value=0.15}
        {arriveTangent=1.1179, interpMode=RCIM_Cubic, leaveTangent=1.1179, tangentMode=RCTM_User, time=0.2042, value=0.6228}
        {arriveTangent=0.0224, interpMode=RCIM_Cubic, leaveTangent=0.0224, tangentMode=RCTM_User, time=1, value=1}
       ]}}})
   &ScaleColor(
    Scale Alpha=&MaskFloatBySpawnGroup(
     Masked Float=&FloatFromCurve(FloatCurve=DataInterface() {propertyValues {Curve {keys [
          {arriveTangent=-3.0638, interpMode=RCIM_Cubic, leaveTangent=-3.0638, tangentMode=RCTM_User, time=0, value=1}
          {tangentMode=RCTM_User, time=1, value=-1}
         ]}}})
     Passthrough Float=&FloatFromCurve(FloatCurve=DataInterface() {propertyValues {Curve {keys [<at 0 1>, {tangentMode=RCTM_User, time=1, value=-0.5}]}}})
     Spawn Group Mask=1
    )
    Scale RGB=<1.2 1.2 1.2>
    ScaleA=true
    ScaleRGB=true
   )
  ]
  rendererClasses [&NiagaraMeshRendererProperties]
  renderers [{rendererClass=&NiagaraMeshRendererProperties, rendererIndex=0}]
 }
 NE_Parent_Mud_Splash001 {
  emitterUpdateScript [
   &EmitterState()
   &Spawn_NDCFortImpacts(Max Spawn Count=5, Min Spawn Count=3, Reader:NiagaraDataInterfaceDataChannelRead=$System.NDCReader)
   &SimpleFilter_NDCFortImpacts(Exclude=false, Max Spawn Count=1, Min Spawn Count=1)
   &LODDistanceFilter_NDCFortImpacts(LODDistance=6000)
  ]
  particleSpawnScript [
   &InitializeParticleFromNDC_FortImpact() {enabled=false}
   &Read_NDC_FortImpacts_Mud()
   &InitializeParticle(
    Color=<rgba 1 1 1 1.2>
    Lifetime Max=0.6
    Lifetime Min=0.3
    Lifetime Mode=::NewEnumerator1
    Position=$Output.InitializeParticleFromNDC_FortImpact.ImpactPoint
    Position Mode=::NewEnumerator1
    Sprite Rotation Mode=::NewEnumerator2
    Sprite Size Mode=::NewEnumerator3
    Sprite UV Mode=::NewEnumerator3
    Uniform Sprite Size=&Multiply_Float(
     A=&RandomRangeFloat(Minimum=40, Maximum=50)
     B=&RemapRange(Input Value To Remap=$Output.InitializeParticleFromNDC_FortImpact.ImpactDamage, Input Min=10, Input Max=100, Output Min=0.9, Output Max=1.1)
    )
    UsePositionOffset=false
    Write Material Random=true
   )
   &SubUVAnimation(
    "Sprite Renderer "=DataInterface() {propertyValues {SpriteRenderer=&self:NE_Parent_Mud_Splash001_1.NiagaraSpriteRendererProperties_1}}
    SubUV Animation Mode:ENiagara_SubUVLookupModeV2=::NewEnumerator2
    UseEndFrame=false
    UseStartFrame=false
   )
   &ShapeLocation(
    Apply To Particle Position=::NewEnumerator1
    MaskSpawnGroup=false
    Offset Coordinate Space=::Local
    Offset Mode=::NewEnumerator0
    Override Seed=false
    Rotation Coordinate Space=::World
    Rotation Mode=::NewEnumerator1
    Sphere Radius=&MaskFloatBySpawnGroup(Masked Float=0.1, Passthrough Float=2, Spawn Group Mask=1)
    UseOwnerScale=true
   ) {module=ShapeLocation001}
   &InitialMeshOrientation(
    Mesh Orientation Mode=::NewEnumerator7
    Rotation=&RandomRangeVector(Minimum=<0 0 0>, Maximum=<0 0 360>)
    Use Rotation Vector=true
   )
   &AddVelocity(Velocity=$Particles.ShapeLocation001.ShapeNormal, Velocity Speed Scale=50)
   &AddVelocity(
    Velocity=$Output.InitializeParticleFromNDC_FortImpact.ImpactNormal
    Velocity Speed Scale=&Multiply_Float(
     A=&RandomRangeFloat(Minimum=100, Maximum=150)
     B=&RemapRange(
      Input Value To Remap=$Output.InitializeParticleFromNDC_FortImpact.ImpactDamage
      Input Min=10
      Input Max=100
      Output Min=0.75
      Output Max=1.25
     )
    )
   ) {module=AddVelocity001}
   &CameraOffset(Camera Offset Amount=15)
  ]
  particleUpdateScript [
   &ParticleState()
   &GravityForce()
   &SolveForcesAndVelocity(Clamp Velocity=false, Limit Acceleration=false)
   &ScaleSpriteSize(Uniform Curve Sprite Scale=DataInterface() {propertyValues {Curve {keys [
        {arriveTangent=4.1785, interpMode=RCIM_Cubic, leaveTangent=4.1785, tangentMode=RCTM_User, time=0, value=0.15}
        {arriveTangent=1.1179, interpMode=RCIM_Cubic, leaveTangent=1.1179, tangentMode=RCTM_User, time=0.2042, value=0.6228}
        {arriveTangent=0.0224, interpMode=RCIM_Cubic, leaveTangent=0.0224, tangentMode=RCTM_User, time=1, value=1}
       ]}}})
   &ScaleColor(
    Scale Alpha=&MaskFloatBySpawnGroup(
     Masked Float=&FloatFromCurve(FloatCurve=DataInterface() {propertyValues {Curve {keys [
          {arriveTangent=-3.0638, interpMode=RCIM_Cubic, leaveTangent=-3.0638, tangentMode=RCTM_User, time=0, value=1}
          {tangentMode=RCTM_User, time=1, value=-1}
         ]}}})
     Passthrough Float=&FloatFromCurve(FloatCurve=DataInterface() {propertyValues {Curve {keys [<at 0.5 1 RCIM_Cubic>, {arriveTangent=-2.0383, interpMode=RCIM_Cubic, leaveTangent=-2.0383, tangentMode=RCTM_User, time=1, value=0}]}}})
     Spawn Group Mask=1
    )
    Scale RGB=<0 0 0>
    ScaleA=true
    ScaleRGB=true
   )
  ]
  rendererClasses [&NiagaraSpriteRendererProperties]
  renderers [{rendererClass=&NiagaraSpriteRendererProperties}]
 }
}
summary {
 systemName=NS_NDCHandler_Impacts_Mud
 emitters [
  {emitterName=Flash, rendererClasses [&NiagaraMeshRendererProperties, &NiagaraSpriteRendererProperties]}
  {bEnabled=false, emitterName=Light, rendererClasses [&NiagaraLightRendererProperties]}
  {bEnabled=false, emitterName=Debug, rendererClasses [&NiagaraSpriteRendererProperties]}
  {emitterName=NE_Parent_Mud_Splash, rendererClasses [&NiagaraMeshRendererProperties]}
  {emitterName=NE_Parent_Mud_DropsCollision, rendererClasses [&NiagaraSpriteRendererProperties]}
  {emitterName=NE_Parent_Mud_Splash001, rendererClasses [&NiagaraSpriteRendererProperties]}
 ]
 userVariables [
  User.SkeletalMesh {
   type=NiagaraDataInterfaceSkeletalMesh
   defaultValue {
    struct=NiagaraExt_VariableValue_DataInterface
    value {dataInterface=&self:NiagaraDataInterfaceSkeletalMesh_0}
   }
  }
  User.SpawnBurst {type=NiagaraBool, defaultValue=true}
 ]
}
